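# UI master material for the Loud modular button, coloring idle, hovered, focused, pressed and disabled states.
# An animated gradient from rotated, panned UVs, sine and easing can blend A/B color pairs for all states but disabled, behind a static switch.
# A shared function blends the states, with optional mixing toward the Slate post-process buffer.
# A rounded-box function supplies fill and outline masks, while alpha-composite blending allows adjustable additive glow.
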Material /SparksUICommon/UILibrary/M_UI_FMCore_Button_Loud_Master
graph {
 properties {blendMode=BLEND_AlphaComposite, materialDomain=MD_UI, pixelDepthOffsetMode=PDOM_Legacy, usedShadingModels=Default Lit}
 nodes [
  n1 = VectorParameter() {group=State colors, parameterName=Color_Disabled, sortPriority=4, defaultValue=<rgba 1 1 1 1>}
  n2 = MaterialFunctionCall!(
   Idle=NamedRerouteUsage()
   Idle Opacity=NamedRerouteUsage()
   Hovered=NamedRerouteUsage()
   Hovered Opacity=NamedRerouteUsage()
   Focused=NamedRerouteUsage()
   Focused Opacity=NamedRerouteUsage()
   Pressed=NamedRerouteUsage()
   Pressed Opacity=NamedRerouteUsage()
   Disabled=@n1.RGB
   Disabled Opacity=@n1.RGB
  ) {materialFunction=/FortUILibrary/Systems/ModularButton/MaterialFunctions/MF_UIKit_ButtonStatesBase_Lerp}
  n3 = LinearInterpolate(
   A=@n2.Color
   B=MaterialFunctionCall!("UVs (Optional)") {materialFunction=/Engine/Functions/UserInterface/GetSlatePost0}.LinearRGB
   Alpha=ScalarParameter() {parameterName=PostProcessMixAmount}
  )
  n4 = Min(
   A=Multiply(A=ScalarParameter() {defaultValue=1, group=Color, parameterName=Brightness Percentage, sliderMax=1, sortPriority=2}, B=@n2.Opacity)
   B=NamedRerouteUsage()
  )
  Multiply(A=ScalarParameter() {defaultValue=1, parameterName=Opacity, sliderMax=1}, B=Multiply(A=VertexColor().A, B=Multiply(A=@n3, B=@n4)))
  Multiply(A=@n4, B=ScalarParameter() {defaultValue=1, group=Color, parameterName=Additive Percentage, sliderMax=1, sortPriority=3})
  n5 = MaterialFunctionCall(
   UV=NamedRerouteUsage()
   Size=AppendVector(A=ScalarParameter() {defaultValue=1, parameterName=Length}, B=ScalarParameter() {defaultValue=1, parameterName=Height})
   Corner Radius=ScalarParameter() {defaultValue=0.3, parameterName=Corner Radius}
   Radius in Pixels=ScalarParameter() {parameterName=Radius in Pixels, sliderMax=1}
   Outline Thickness=ScalarParameter() {defaultValue=1, parameterName=Outline Thickness}
  ) {materialFunction=&MF_UI_CoreUI_RoundedMask}
  NamedRerouteDeclaration(None=@n5.Shape Mask) {name=Fill_Box, nodeColor=<rgba 0 0 0 1>}
  NamedRerouteDeclaration(None=MaterialFunctionCall() {materialFunction=&GetUserInterfaceUV}.Normalized UV) {name=Normalized UV, nodeColor=<rgba 0 0 0 1>}
  n6 = VectorParameter() {group=State colors, parameterName=Color_Focused_A, sortPriority=2, defaultValue=<rgba 1 1 1 1>}
  n7 = VectorParameter() {group=State colors, parameterName=Color_Focused_B, sortPriority=2, defaultValue=<rgba 1 1 1 1>}
  n8 = NamedRerouteUsage()
  NamedRerouteDeclaration(None=StaticSwitchParameter(True=LinearInterpolate(A=@n6.RGB, B=@n7.RGB, Alpha=@n8), False=@n6.RGB) {defaultValue=true, parameterName=HasGradientColors}) {name=ColorFocused, nodeColor=<rgba 0.6875 1 0 1>}
  NamedRerouteDeclaration(None=StaticSwitchParameter(True=LinearInterpolate(A=@n6.A, B=@n7.A, Alpha=@n8), False=@n6.A) {defaultValue=true, parameterName=HasGradientColors}) {name=AlphaFocused, nodeColor=<rgba 0.0078 1 0 1>}
  n9 = VectorParameter() {group=State colors, parameterName=Color_Pressed_A, sortPriority=3, defaultValue=<rgba 1 1 1 1>}
  n10 = VectorParameter() {group=State colors, parameterName=Color_Pressed_B, sortPriority=3, defaultValue=<rgba 1 1 1 1>}
  n11 = NamedRerouteUsage()
  NamedRerouteDeclaration(None=StaticSwitchParameter(True=LinearInterpolate(A=@n9.RGB, B=@n10.RGB, Alpha=@n11), False=@n9.RGB) {defaultValue=true, parameterName=HasGradientColors}) {name=ColorPressed, nodeColor=<rgba 0 0.3906 1 1>}
  NamedRerouteDeclaration(None=StaticSwitchParameter(True=LinearInterpolate(A=@n9.A, B=@n10.A, Alpha=@n11), False=@n9.A) {defaultValue=true, parameterName=HasGradientColors}) {name=AlphaPressed, nodeColor=<rgba 0.9688 0 1 1>}
  n12 = MaterialFunctionCall(
   UVs=NamedRerouteUsage()
   Rotation Center=AppendVector(
    A=ScalarParameter() {defaultValue=0.5, parameterName=Gradient Center X, sliderMax=1}
    B=ScalarParameter() {defaultValue=0.5, parameterName=Gradient Center Y, sliderMax=1}
   )
   "Rotation Angle (0-1)"=ScalarParameter() {defaultValue=-0.05, parameterName=Gradient Rotation, sliderMax=1}
  ) {materialFunction=&CustomRotator}
  n13 = Time()
  n14 = AppendVector(A=ScalarParameter() {defaultValue=-0.25, parameterName=PanSpeedX}, B=Constant())
  n15 = ComponentMask(None=Sine(None=Panner(Coordinate=@n12.Rotated Values, Time=@n13, Speed=@n14)) {period=2}) {r=true}
  n16 = MaterialFunctionCall("Value (0-1)"=Multiply(A=Add(A=@n15)) {constB=0.5}) {materialFunction=&MF_UI_EaseCurves}
  NamedRerouteDeclaration(None=Saturate(None=OneMinus(None=@n16.EaseOutSine))) {name=GradientMask, nodeColor=<rgba 0 0 0 1>}
  n17 = VectorParameter() {group=State colors, parameterName=Color_Idle_A, sortPriority=0, defaultValue=<rgba 0.6297 0 1 1>}
  n18 = VectorParameter() {group=State colors, parameterName=Color_Idle_B, sortPriority=0, defaultValue=<rgba 0.3116 0 0.4948 1>}
  n19 = NamedRerouteUsage()
  NamedRerouteDeclaration(None=StaticSwitchParameter(True=LinearInterpolate(A=@n17.RGB, B=@n18.RGB, Alpha=@n19), False=@n17.RGB) {defaultValue=true, parameterName=HasGradientColors}) {name=ColorIdle, nodeColor=<rgba 0.6172 1 0 1>}
  NamedRerouteDeclaration(None=@n5.Outline Mask) {name=Outline_Box, nodeColor=<rgba 0.1953 1 0 1>}
  NamedRerouteDeclaration(None=StaticSwitchParameter(True=LinearInterpolate(A=@n17.A, B=@n18.A, Alpha=@n19), False=@n17.A) {defaultValue=true, parameterName=HasGradientColors}) {name=AlphaIdle, nodeColor=<rgba 0.5 1 0 1>}
  n20 = VectorParameter() {group=State colors, parameterName=Color_Hovered_A, sortPriority=1, defaultValue=<rgba 1 1 1 1>}
  n21 = VectorParameter() {group=State colors, parameterName=Color_Hovered_B, sortPriority=1, defaultValue=<rgba 1 1 1 1>}
  n22 = NamedRerouteUsage()
  NamedRerouteDeclaration(None=StaticSwitchParameter(True=LinearInterpolate(A=@n20.RGB, B=@n21.RGB, Alpha=@n22), False=@n20.RGB) {defaultValue=true, parameterName=HasGradientColors}) {name=ColorHovered, nodeColor=<rgba 0 0.1562 1 1>}
  NamedRerouteDeclaration(None=StaticSwitchParameter(True=LinearInterpolate(A=@n20.A, B=@n21.A, Alpha=@n22), False=@n20.A) {defaultValue=true, parameterName=HasGradientColors}) {name=AlphaHovered, nodeColor=<rgba 1 0.375 0 1>}
 ]
}
